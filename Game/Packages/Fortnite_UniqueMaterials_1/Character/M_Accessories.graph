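# Masked, two-sided skeletal-mesh material for character accessories, with hat textures as defaults.
# Diffuse, optionally tinted by two mask-driven accessory colors, passes through a fresnel function with optional supercharge glow emissive, then desaturation.
# Packed specular masks drive metallic, specular and roughness, while shared functions add res-in opacity masking and black-hole world position offset.
# All attributes finish through the Fortnite base function and an accessories tech-art override function.
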Material /Game/Packages/Fortnite_UniqueMaterials_1/Character/M_Accessories
graph {
 properties {
  bUseMaterialAttributes=true
  bUsedWithMeshDeformer=true
  bUsedWithMorphTargets=true
  bUsedWithSkeletalMesh=true
  blendMode=BLEND_Masked
  pixelDepthOffsetMode=PDOM_Legacy
  refractionMethod=RM_IndexOfRefraction
  twoSided=true
  usedShadingModels=Default Lit
 }
 nodes [
  n1 = VectorParameter() {group=Super Charge, parameterName=AverageHandPosition, defaultValue=<rgba 0 0 0 1>}
  n2 = ScalarParameter() {defaultValue=1, group=Super Charge, parameterName=SuperChargedFlickerBias}
  n3 = ScalarParameter() {group=Super Charge, parameterName=SuperChargedEffect}
  n4 = StaticBool()
  n5 = TextureSampleParameter2D() {group=Diffuse, parameterName=Diffuse, texture=/Game/Accessories/Hats/Textures/FootballHelmet/T_Hat_Football_D}
  n6 = VectorParameter() {group=Colorize, parameterName=AccessoryColor3, defaultValue=<rgba 1 1 1 0>}
  n7 = VectorParameter() {group=Colorize, parameterName=AccessoryColor2, defaultValue=<rgba 1 1 1 0>}
  n8 = TextureSampleParameter2D() {
   desc="Red = accessory color 2 mask - black and white indicates how much colorization ( * accessory 2 alpha), Green same but for acceccory color 3, blue is the value of the colorized values"
   group=Colorize
   parameterName=ClothingAndSkinMasks
   samplerType=SAMPLERTYPE_Masks
   texture=/Game/Characters/Player/Male/LegacyAssets/Male_Avg_Torsos/M_Avg_Torso_FastFoodWorker_01/Textures/T_M_Avg_Torso_FastFoodWorker_01_RGB
  }
  n9 = Multiply(A=LinearInterpolate(A=@n6.RGB, B=@n7.RGB, Alpha=@n8.R), B=@n8.B)
  n10 = Clamp(None=Add(A=Multiply(A=@n7.A, B=@n8.R), B=Multiply(A=@n6.A, B=@n8.G)))
  n11 = MaterialFunctionCall!(
   DiffuseFresnelAmount=ScalarParameter() {defaultValue=0.75, group=Diffuse, parameterName=DiffuseFresnelAmount}
   Emissive=StaticSwitchParameter(
    True=MaterialFunctionCall!(AverageHandPosition=@n1.RGB, SuperChargedFlickerBias=@n2, SuperChargedEffect=@n3, UseDistanceCalc=@n4, Emissive) {materialFunction=/Game/Characters/Fortress_CharacterShader/SuperChargedGlow}.Result
    False=Constant()
   ) {group=Super Charge, parameterName=Use Super Charged Effect}
   Base Color=StaticSwitchParameter(True=LinearInterpolate(A=@n5.RGB, B=@n9, Alpha=@n10), False=@n5.RGB) {group=Colorize, parameterName=Colorize}
  ) {materialFunction=/Game/Packages/Fortress_SharedMaterials/Base_Material_Functions/FresnelFunction}
  n12 = Desaturation(None=@n11.d, Fraction=ScalarParameter() {group=Code, parameterName=DesaturationPercentage})
  n13 = TextureSampleParameter2D() {parameterName=SpecularMasks, samplerType=SAMPLERTYPE_Masks, texture=&T_Fortnite_Default_S}
  n14 = Multiply(A=@n13.G, B=ScalarParameter() {defaultValue=1, group=Specular, parameterName=Metallic})
  n15 = Multiply(A=@n13.R, B=ScalarParameter() {defaultValue=1, group=Specular, parameterName=Specular})
  n16 = LinearInterpolate(
   A=ScalarParameter() {defaultValue=0.1, group=Specular, parameterName=SpecRoughnessMin}
   B=ScalarParameter() {defaultValue=1, group=Specular, parameterName=SpecRoughnessMax}
   Alpha=@n13.B
  )
  n17 = MaterialFunctionCall!(Animation Gradient, Rez In Percentage) {materialFunction=/Game/Packages/Fortress_SharedMaterials/Base_Material_Functions/CharacterResInEffect}
  n18 = TextureSampleParameter2D() {
   group=Normals
   parameterName=Normals
   samplerType=SAMPLERTYPE_Normal
   texture=/Game/Accessories/Hats/Textures/FootballHelmet/T_Hat_Football_N
  }
  n19 = MaterialFunctionCall!(WorldPostionOfBlackHole, World Position) {materialFunction=/Game/Packages/Fortnite_UniqueMaterials_1/BlackHoleFunction}
  MaterialFunctionCall(In=MaterialFunctionCall!(
    Material attributes=MakeMaterialAttributes!(
     BaseColor=@n12
     Metallic=@n14
     Specular=@n15
     Roughness=@n16
     Anisotropy
     EmissiveColor=@n11.Emissive
     Opacity
     OpacityMask=@n17.Modulative Opacity
     Normal=@n18.RGB
     Tangent
     WorldPositionOffset=@n19.Result
     SubsurfaceColor
     ClearCoat
     ClearCoatRoughness
     AmbientOcclusion
     Refraction
     CustomizedUVs_0
     CustomizedUVs_1
     CustomizedUVs_2
     CustomizedUVs_3
     CustomizedUVs_4
     CustomizedUVs_5
     CustomizedUVs_6
     CustomizedUVs_7
     PixelDepthOffset
     ShadingModel
     Displacement
    )
    IsGameplayAsset=StaticBool()
   ) {materialFunction=/Game/Packages/Fortress_SharedMaterials/Base_Material_Functions/Fortnite_BaseFunction}.Material Attributes) {materialFunction=&MF_TechArtOverrides_accessories}
 ]
}
