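# Two-sided Default Lit character material for the Blueprint Architect, enabled for skeletal meshes, morph targets and mesh deformers.
# It exposes Diffuse and Normal texture parameters with football helmet hat defaults, plus an EmissiveMask multiplied by an EmissiveMultiply scalar.
# A non-parameter mask sample uses the building tools' ArchitectTools texture, so material instances cannot swap that map.
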Material /Game/Packages/Fortnite_UniqueMaterials_1/Character/M_BluePrintArchitect
graph {
 properties {
  bUsedWithMeshDeformer=true
  bUsedWithMorphTargets=true
  bUsedWithSkeletalMesh=true
  pixelDepthOffsetMode=PDOM_Legacy
  twoSided=true
  usedShadingModels=Default Lit
 }
 nodes [
  Multiply(
   A=TextureSampleParameter2D() {parameterName=EmissiveMask, texture=/Game/Accessories/Hats/Textures/Viking/T_Hat_Viking_M}.RGB
   B=ScalarParameter() {parameterName=EmissiveMultiply}
  )
  TextureSampleParameter2D() {parameterName=Diffuse, texture=/Game/Accessories/Hats/Textures/FootballHelmet/T_Hat_Football_D}
  TextureSampleParameter2D() {parameterName=Normal, samplerType=SAMPLERTYPE_Normal, texture=/Game/Accessories/Hats/Textures/FootballHelmet/T_Hat_Football_N}
  TextureSample() {samplerType=SAMPLERTYPE_Masks, texture=/Game/Weapons/FORT_BuildingTools/Textures/T_ArchitectTools_S}
 ]
}
